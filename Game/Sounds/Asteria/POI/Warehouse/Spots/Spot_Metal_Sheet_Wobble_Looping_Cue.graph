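# Looping ambient spot cue for the Warehouse POI that plays intermittent metal sheet wobbles and creaks.
# Each loop waits a random delay, then mixes a low-pitched, quieter random Tilted Towers metal movement with a random JunkYard wobble faded out by an Enveloper.
# Modulators randomize pitch on both layers.
# Overridden box attenuation adds occlusion and listener focus, concurrency blocks excess instances, and a bus send feeds exterior ambience.
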SoundCue /Game/Sounds/Asteria/POI/Warehouse/Spots/Spot_Metal_Sheet_Wobble_Looping_Cue
root=Looping_0
nodes [
 {id=Delay_0, children [Mixer_0], properties {delayMax=20, delayMin=6}}
 {
  id=Enveloper_0
  children [Modulator_0]
  properties {
   pitchCurve {editorCurveData {keys [{time=0}]}}
   volumeCurve {editorCurveData {keys [<at 0 1>, {arriveTangent=-0.2, interpMode=RCIM_Cubic, leaveTangent=-0.2, time=1, value=1}, <at 5 0>]}}
  }
 }
 {id=Looping_0, children [Delay_0]}
 {id=Mixer_0, children [Modulator_1,Enveloper_0], properties {inputVolume [1,1]}}
 {id=Modulator_0, children [Random_0], properties {pitchMin=0.9}}
 {id=Modulator_1, children [Random_1], properties {pitchMax=0.6, pitchMin=0.5, volumeMax=0.5, volumeMin=0.4}}
 {id=Random_0, children [WavePlayer_10, WavePlayer_11, WavePlayer_12, WavePlayer_13], properties {weights [1,1,1,1]}}
 {id=Random_1, children [WavePlayer_0,WavePlayer_1,WavePlayer_2,WavePlayer_8,WavePlayer_9,WavePlayer_3], properties {weights [1,1,1,1,1,1]}}
 {id=WavePlayer_0, properties {soundWaveAssetPtr=/Game/Sounds/Artemis/POI/Tilted_Towers/Spots/Spot_Tilted_Towers_Metal_Movement_06}}
 {id=WavePlayer_1, properties {soundWaveAssetPtr=/Game/Sounds/Artemis/POI/Tilted_Towers/Spots/Spot_Tilted_Towers_Metal_Movement_01}}
 {id=WavePlayer_10, properties {soundWaveAssetPtr=/Game/Sounds/Asteria/POI/JunkYard/Spots/Spot_Metal_Sheet_Wobble_01}}
 {id=WavePlayer_11, properties {soundWaveAssetPtr=/Game/Sounds/Asteria/POI/JunkYard/Spots/Spot_Metal_Sheet_Wobble_02}}
 {id=WavePlayer_12, properties {soundWaveAssetPtr=/Game/Sounds/Asteria/POI/JunkYard/Spots/Spot_Metal_Sheet_Wobble_03}}
 {id=WavePlayer_13, properties {soundWaveAssetPtr=/Game/Sounds/Asteria/POI/JunkYard/Spots/Spot_Metal_Sheet_Wobble_04}}
 {id=WavePlayer_2, properties {soundWaveAssetPtr=/Game/Sounds/Artemis/POI/Tilted_Towers/Spots/Spot_Tilted_Towers_Metal_Movement_02}}
 {id=WavePlayer_3, properties {soundWaveAssetPtr=/Game/Sounds/Artemis/POI/Tilted_Towers/Spots/Spot_Tilted_Towers_Metal_Movement_05}}
 {id=WavePlayer_8, properties {soundWaveAssetPtr=/Game/Sounds/Artemis/POI/Tilted_Towers/Spots/Spot_Tilted_Towers_Metal_Movement_03}}
 {id=WavePlayer_9, properties {soundWaveAssetPtr=/Game/Sounds/Artemis/POI/Tilted_Towers/Spots/Spot_Tilted_Towers_Metal_Movement_04}}
]
properties {
 attenuationSettings=/Game/Sounds/Asteria/POI/JunkYard/Spots/Spot_Metal_Movements_Distant_Att
 bEnableSubmixSends=false
 bOverrideAttenuation=true
 bOverrideConcurrency=true
 maxDistance=2251.0955
 soundClassObject=&Fort_Ambient_OneShots
 volumeMultiplier=0.601
 attenuationOverrides {
  attenuationShape=Box
  bEnableFocusInterpolation=true
  bEnableListenerFocus=true
  bEnableOcclusion=true
  falloffDistance=2000
  focusAttackInterpSpeed=0.5
  focusReleaseInterpSpeed=0.5
  focusVolumeAttenuation=1.2
  occlusionLowPassFilterFrequency=1000
  occlusionTraceChannel=ECC_GameTraceChannel11
  occlusionVolumeAttenuation=0.2
  reverbWetLevelMin=0.6
  attenuationShapeExtents=<251.0955 39.6181 175>
 }
 busSends [{audioBus=null, minSendLevel=0.6, sendLevel=1, soundSourceBus=&AmbientAudio_Exterior_SourceBus}]
 concurrencyOverrides {maxCount=2, resolutionRule=PreventNew}
}
